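; Anniversary balloon prop for Athena that exists only while the anniversary event is enabled and plays a pop effect when killed.
; On BeginPlay and again on WorldReady, the authority reads Default.Anniversary.Enabled from the AthenaGameData curve table and destroys the actor if the found value is 0.5 or lower.
; Server death calls Play FX, which uses a DoOnce to spawn the Niagara pop and sound at the actor's location.
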(event EventBeginPlay
  ;[1 1 1] evaluate default.birthdaycake
  (Networking|SwitchHasAuthority
    (:Authority
      (bind _outxy (DataTable|EvaluateCurveTableRow "/Game/Athena/Balance/DataTables/AthenaGameData.AthenaGameData" "Default.Anniversary.Enabled")
        (:RowFound
          (if (not (> _outxy 0.5))
            (Actor|DestroyActor)))))))

(event EventOnWorldReady
  ;[1 1 1] evaluate default.birthdaycake
  (Networking|SwitchHasAuthority
    (:Authority
      (bind _outxy (DataTable|EvaluateCurveTableRow "/Game/Athena/Balance/DataTables/AthenaGameData.AthenaGameData" "Default.Anniversary.Enabled")
        (:RowFound
          (if (not (> _outxy 0.5))
            (Actor|DestroyActor)))))))

(event EventOnDeathServer
  (params (Damage :type Float%20%28single-precision%29) (DamageTags :type Gameplay%20Tag%20Container%20Structure%20%28by%20ref%29) (Momentum :type Vector :default "0, 0, 0") (HitInfo :type Hit%20Result%20Structure%20%28by%20ref%29) (InstigatedBy :type Controller%20Object%20Reference) (DamageCauser :type Actor%20Object%20Reference) (EffectContext :type Gameplay%20Effect%20Context%20Handle%20Structure))
  ;[1 1 1] Play particles when Balloon pops
  (CallFunction|PlayFX))

(event Custom|Play%20FX
  (bind _self self)
  (bind _returnvalue (Transformation|GetActorLocation _self))
  ;[1 1 1] Play particles when Balloon pops
  (Utilities|FlowControl|DoOnce false
    (:Completed
      (Niagara|SpawnSystematLocation "/Game/Environments/Props/BirthdayBalloons/FX/NS_Balloons_Birthday_Pop_Tri.NS_Balloons_Birthday_Pop_Tri" _returnvalue)
      (Audio|PlaySoundatLocation (Variables|Default|GetSound) _returnvalue))))

(fn ConstructionScript)
