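; Anniversary 2018 birthday cake prop that marks players standing near it, apparently for an event quest.
; On BeginPlay and world ready, the authority reads the Default.Anniversary.Enabled curve table row and destroys the cake if its value is not above 0.5.
; CakeDanceZone overlaps apply NearACake or NoLongerNearACake gameplay effects to player-controlled pawns, only where cosmetic events cannot run, such as dedicated servers.
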(event EventBeginPlay
  ;[1 1 1] evaluate default.birthdaycake
  (Networking|SwitchHasAuthority
    (:Authority
      (bind _outxy (DataTable|EvaluateCurveTableRow "/Game/Athena/Balance/DataTables/AthenaGameData.AthenaGameData" "Default.Anniversary.Enabled")
        (:RowFound
          (if (not (> _outxy 0.5))
            (Actor|DestroyActor)))))))

(event EventOnWorldReady
  ;[1 1 1] evaluate default.birthdaycake
  (Networking|SwitchHasAuthority
    (:Authority
      (bind _outxy (DataTable|EvaluateCurveTableRow "/Game/Athena/Balance/DataTables/AthenaGameData.AthenaGameData" "Default.Anniversary.Enabled")
        (:RowFound
          (if (not (> _outxy 0.5))
            (Actor|DestroyActor)))))))

(event OnComponentBeginOverlap(CakeDanceZone)
  (params (OverlappedComponent :type Primitive%20Component%20Object%20Reference) (OtherActor :type Actor%20Object%20Reference) (OtherComp :type Primitive%20Component%20Object%20Reference) (OtherBodyIndex :type Integer) (bFromSweep :type Boolean) (SweepResult :type Hit%20Result%20Structure%20%28by%20ref%29))
  (CallFunction|AddDanceZoneGE OtherActor))

(event OnComponentEndOverlap(CakeDanceZone)
  (params (OverlappedComponent :type Primitive%20Component%20Object%20Reference) (OtherActor :type Actor%20Object%20Reference) (OtherComp :type Primitive%20Component%20Object%20Reference) (OtherBodyIndex :type Integer))
  (CallFunction|RemoveDanceZoneGE OtherActor))

(fn AddDanceZoneGE
  (params (Actor%20Entering%20Zone :type Actor%20Object%20Reference))
  (Networking|CanExecuteCosmeticEvents
    (:False
      (bind _aspawn (Utilities|Casting|CastToPawn Actor%20Entering%20Zone))
      (if (Pawn|IsPlayerControlled _aspawn)
        (Fortnite|ApplyGameplayEffectsToActor (Variables|GetActor%20Entering%20Zone) (Utilities|Array|MakeArray "/Game/Missions/Quests/ReactiveQuests/Events/Anniversary2018/GEs/GE_Quest_NearACake.GE_Quest_NearACake_C") (Transformation|GetWorldLocation (Variables|Default|GetCakeDanceZone)))))))

(fn RemoveDanceZoneGE
  (params (Actor%20Leaving%20Zone :type Actor%20Object%20Reference))
  (Networking|CanExecuteCosmeticEvents
    (:False
      (bind _aspawn (Utilities|Casting|CastToPawn Actor%20Leaving%20Zone))
      (if (Pawn|IsPlayerControlled _aspawn)
        (Fortnite|ApplyGameplayEffectsToActor (Variables|GetActor%20Leaving%20Zone) (Utilities|Array|MakeArray "/Game/Missions/Quests/ReactiveQuests/Events/Anniversary2018/GEs/GE_Quest_NoLongerNearACake.GE_Quest_NoLongerNearACake_C") (Transformation|GetWorldLocation (Variables|Default|GetCakeDanceZone)))))))

(fn ConstructionScript)
